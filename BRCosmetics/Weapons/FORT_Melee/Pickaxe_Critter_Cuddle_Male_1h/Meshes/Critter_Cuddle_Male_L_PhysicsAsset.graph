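# Physics asset for the Critter Cuddle Male pickaxe's L mesh, giving capsule bodies and joints to its jaw, leaf and strap bones.
# The jaw hangs off the head master bone with tight limits, while three leaf chains, an upright leaf pair and a four-link strap chain branch from the straps master bone.
# Root, handle and master joints rotate freely; chain joints have angular limits, stiffest on the upright leaf and widest at the strap base.
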PhysicsAsset /BRCosmetics/Weapons/FORT_Melee/Pickaxe_Critter_Cuddle_Male_1h/Meshes/Critter_Cuddle_Male_L_PhysicsAsset
bodies {
 Melee_TwoHanded_Handle [{shapeType=Capsule, length=2, radius=2, shapeName=Melee_TwoHanded_Handle_capsule}]
 b_AR_Root [{shapeType=Capsule, length=2, radius=2, shapeName=b_AR_Root_capsule}]
 dyn_jaw_jnt [{shapeType=Capsule, length=6, radius=2.5, shapeName=dyn_jaw_jnt_capsule, center=<-5.2854 0 -0.4607>, rotation=<rot -79.9999 0 0>}]
 dyn_leaf_A_1_jnt [{
   shapeType=Capsule
   length=1.5
   radius=1
   shapeName=dyn_leaf_A_1_jnt_capsule
   center=<1.6658 -0.1436 -0.0413>
   rotation=<rot -58.5219 106.7463 -109.4327>
  }]
 dyn_leaf_A_2_jnt [{shapeType=Capsule, length=1, radius=1, shapeName=dyn_leaf_A_2_jnt_capsule, center=<1.4402 0 -0.2552>, rotation=<rot -79.994 180 179.9999>}]
 dyn_leaf_A_3_jnt [{
   shapeType=Capsule
   length=1
   radius=1
   shapeName=dyn_leaf_A_3_jnt_capsule
   center=<1.556 0 -0.2744>
   rotation=<rot -79.9983 179.9999 179.9999>
  }]
 dyn_leaf_B_1_jnt [{shapeType=Capsule, length=0.75, radius=1, shapeName=dyn_leaf_B_1_jnt_capsule, center=<1.4605 0.0001 0>, rotation=<rot -90 0 0>}]
 dyn_leaf_B_2_jnt [{shapeType=Capsule, length=0.75, radius=1, shapeName=dyn_leaf_B_2_jnt_capsule, center=<1.4057 0 0>, rotation=<rot 90 -0.0011 -0.001>}]
 dyn_leaf_B_3_jnt [{shapeType=Capsule, length=1, radius=1, shapeName=dyn_leaf_B_3_jnt_capsule, center=<0 -1.6553 0>, rotation=<rot 39.9997 -0.0001 90.0008>}]
 dyn_leaf_C_1_jnt [{shapeType=Capsule, length=0.5, radius=1, shapeName=dyn_leaf_C_1_jnt_capsule, center=<1.2791 0 0.0001>, rotation=<rot -90 -7.5524 7.5525>}]
 dyn_leaf_C_2_jnt [{shapeType=Capsule, length=0.5, radius=1, shapeName=dyn_leaf_C_2_jnt_capsule, center=<1.7041 0 0>, rotation=<rot -90 1.8985 -1.8984>}]
 dyn_leaf_C_3_jnt [{
   shapeType=Capsule
   length=2.25
   radius=1
   shapeName=dyn_leaf_C_3_jnt_capsule
   center=<2.2445 -0.0001 0>
   rotation=<rot -90 -179.9802 179.9803>
  }]
 dyn_leaf_up_1_jnt [{shapeType=Capsule, length=0.5, radius=1, shapeName=dyn_leaf_up_1_jnt_capsule, center=<1.1519 0 0>, rotation=<rot -90 -0.0198 0.0198>}]
 dyn_leaf_up_2_jnt [{
   shapeType=Capsule
   length=1
   radius=1
   shapeName=dyn_leaf_up_2_jnt_capsule
   center=<1.4916 -0.2278 -0.5429>
   rotation=<rot -69.9993 179.9998 180>
  }]
 dyn_strap_1_jnt [{shapeType=Capsule, length=2, radius=1, shapeName=dyn_strap_1_jnt_capsule, center=<1.7672 0 0>, rotation=<rot 90 0 0>}]
 dyn_strap_2_jnt [{shapeType=Capsule, length=2, radius=1, shapeName=dyn_strap_2_jnt_capsule, center=<2.1617 0 0>, rotation=<rot 90 0.0099 0.0099>}]
 dyn_strap_3_jnt [{shapeType=Capsule, length=2, radius=1, shapeName=dyn_strap_3_jnt_capsule, center=<1.8872 0 0>, rotation=<rot 90 0.0074 0.0074>}]
 dyn_strap_4_jnt [{shapeType=Capsule, length=2, radius=1, shapeName=dyn_strap_4_jnt_capsule, center=<2.1218 0 0>, rotation=<rot 90 0.0074 0.0074>}]
 master_head_jnt [{shapeType=Capsule, length=2, radius=2, shapeName=master_head_jnt_capsule}]
 master_straps_jnt [{shapeType=Capsule, length=2, radius=2, shapeName=master_straps_jnt_capsule}]
}
constraints [
 {
  bone1Name=Melee_TwoHanded_Handle
  bone2Name=b_AR_Root
  swing1LimitDegrees=45
  swing1Motion=Free
  swing2LimitDegrees=45
  swing2Motion=Free
  twistLimitDegrees=45
  twistMotion=Free
 }
 {
  bone1Name=master_head_jnt
  bone2Name=Melee_TwoHanded_Handle
  swing1LimitDegrees=45
  swing1Motion=Free
  swing2LimitDegrees=45
  swing2Motion=Free
  twistLimitDegrees=45
  twistMotion=Free
 }
 {bone1Name=dyn_jaw_jnt, bone2Name=master_head_jnt, swing1LimitDegrees=10, swing2LimitDegrees=2, twistLimitDegrees=10}
 {
  bone1Name=master_straps_jnt
  bone2Name=Melee_TwoHanded_Handle
  swing1LimitDegrees=45
  swing1Motion=Free
  swing2LimitDegrees=45
  swing2Motion=Free
  twistLimitDegrees=45
  twistMotion=Free
 }
 {bone1Name=dyn_leaf_C_1_jnt, bone2Name=master_straps_jnt, swing1LimitDegrees=20, swing2LimitDegrees=20, twistLimitDegrees=10}
 {bone1Name=dyn_leaf_C_2_jnt, bone2Name=dyn_leaf_C_1_jnt, swing1LimitDegrees=20, swing2LimitDegrees=20, twistLimitDegrees=10}
 {bone1Name=dyn_leaf_C_3_jnt, bone2Name=dyn_leaf_C_2_jnt, swing1LimitDegrees=20, swing2LimitDegrees=20, twistLimitDegrees=10}
 {bone1Name=dyn_leaf_up_1_jnt, bone2Name=master_straps_jnt, swing1LimitDegrees=5, swing2LimitDegrees=5, twistLimitDegrees=5}
 {bone1Name=dyn_leaf_up_2_jnt, bone2Name=dyn_leaf_up_1_jnt, swing1LimitDegrees=5, swing2LimitDegrees=5, twistLimitDegrees=5}
 {bone1Name=dyn_strap_1_jnt, bone2Name=master_straps_jnt, swing1LimitDegrees=90, swing2LimitDegrees=90, twistLimitDegrees=10}
 {bone1Name=dyn_strap_2_jnt, bone2Name=dyn_strap_1_jnt, swing1LimitDegrees=20, swing2LimitDegrees=20, twistLimitDegrees=10}
 {bone1Name=dyn_strap_3_jnt, bone2Name=dyn_strap_2_jnt, swing1LimitDegrees=20, swing2LimitDegrees=20, twistLimitDegrees=10}
 {bone1Name=dyn_strap_4_jnt, bone2Name=dyn_strap_3_jnt, swing1LimitDegrees=20, swing2LimitDegrees=20, twistLimitDegrees=10}
 {bone1Name=dyn_leaf_B_1_jnt, bone2Name=master_straps_jnt, swing1LimitDegrees=20, swing2LimitDegrees=20, twistLimitDegrees=10}
 {bone1Name=dyn_leaf_B_2_jnt, bone2Name=dyn_leaf_B_1_jnt, swing1LimitDegrees=20, swing2LimitDegrees=20, twistLimitDegrees=10}
 {bone1Name=dyn_leaf_B_3_jnt, bone2Name=dyn_leaf_B_2_jnt, swing1LimitDegrees=20, swing2LimitDegrees=20, twistLimitDegrees=10}
 {bone1Name=dyn_leaf_A_1_jnt, bone2Name=master_straps_jnt, swing1LimitDegrees=20, swing2LimitDegrees=20, twistLimitDegrees=10}
 {bone1Name=dyn_leaf_A_2_jnt, bone2Name=dyn_leaf_A_1_jnt, swing1LimitDegrees=20, swing2LimitDegrees=20, twistLimitDegrees=10}
 {bone1Name=dyn_leaf_A_3_jnt, bone2Name=dyn_leaf_A_2_jnt, swing1LimitDegrees=20, swing2LimitDegrees=20, twistLimitDegrees=10}
]
