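# Physics asset for the right-hand Critter Cuddle Male one-handed pickaxe, giving capsule bodies to the handle, root, bow, chain and ring bones.
# The handle joins the root and bow_master through free angular constraints.
# A three-link dyn_bow_chain hangs from bow_master with tight limits, beside two bowtie lobes with much wider limits.
# The dyn_ring body attaches directly to the handle with the narrowest limits.
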PhysicsAsset /BRCosmetics/Weapons/FORT_Melee/Pickaxe_Critter_Cuddle_Male_1h/Meshes/Critter_Cuddle_Male_R_PhysicsAsset
bodies {
 Melee_TwoHanded_Handle [{shapeType=Capsule, length=2, radius=2, shapeName=Melee_TwoHanded_Handle_capsule}]
 b_AR_Root [{shapeType=Capsule, length=2, radius=2, shapeName=b_AR_Root_capsule}]
 bow_master [{shapeType=Capsule, length=2, radius=2, shapeName=bow_master_capsule}]
 dyn_bow_chain_1 [{
   shapeType=Capsule
   length=1.25
   radius=0.5
   shapeName=dyn_bow_chain_1_capsule
   center=<1.212 0 -0.4911>
   rotation=<rot -69.9933 -179.9999 180>
  }]
 dyn_bow_chain_2 [{
   shapeType=Capsule
   length=0.75
   radius=0.75
   shapeName=dyn_bow_chain_2_capsule
   center=<0.8749 0 -0.4411>
   rotation=<rot -59.9959 -179.9999 180>
  }]
 dyn_bow_chain_3 [{
   shapeType=Capsule
   length=0.75
   radius=1
   shapeName=dyn_bow_chain_3_capsule
   center=<1.1931 0.0002 -0.6889>
   rotation=<rot 60.0001 0 -0.0001>
  }]
 dyn_bowtie_L [{shapeType=Capsule, length=2.25, radius=2.75, shapeName=dyn_bowtie_L_capsule, center=<0.7186 3.1042 -1.4311>, rotation=<rot 0 0 -60.0003>}]
 dyn_bowtie_R [{shapeType=Capsule, length=1.75, radius=3, shapeName=dyn_bowtie_R_capsule, center=<0.0001 -3.893 -0.7132>, rotation=<rot 0 0 79.998>}]
 dyn_ring [{
   shapeType=Capsule
   length=2
   radius=2
   shapeName=dyn_ring_capsule
   center=<3.0309 0.0001 -0.5345>
   rotation=<rot -79.9997 179.9999 -179.9999>
  }]
}
constraints [
 {
  bone1Name=Melee_TwoHanded_Handle
  bone2Name=b_AR_Root
  swing1LimitDegrees=45
  swing1Motion=Free
  swing2LimitDegrees=45
  swing2Motion=Free
  twistLimitDegrees=45
  twistMotion=Free
 }
 {
  bone1Name=bow_master
  bone2Name=Melee_TwoHanded_Handle
  swing1LimitDegrees=45
  swing1Motion=Free
  swing2LimitDegrees=45
  swing2Motion=Free
  twistLimitDegrees=45
  twistMotion=Free
 }
 {bone1Name=dyn_bow_chain_1, bone2Name=bow_master, swing1LimitDegrees=10, swing2LimitDegrees=10, twistLimitDegrees=5}
 {bone1Name=dyn_bow_chain_2, bone2Name=dyn_bow_chain_1, swing1LimitDegrees=10, swing2LimitDegrees=10, twistLimitDegrees=5}
 {bone1Name=dyn_bow_chain_3, bone2Name=dyn_bow_chain_2, swing1LimitDegrees=10, swing2LimitDegrees=10, twistLimitDegrees=5}
 {bone1Name=dyn_bowtie_L, bone2Name=bow_master, swing1LimitDegrees=45, swing2LimitDegrees=45, twistLimitDegrees=45}
 {bone1Name=dyn_bowtie_R, bone2Name=bow_master, swing1LimitDegrees=45, swing2LimitDegrees=45, twistLimitDegrees=45}
 {bone1Name=dyn_ring, bone2Name=Melee_TwoHanded_Handle, swing1LimitDegrees=5, swing2LimitDegrees=1, twistLimitDegrees=5}
]
